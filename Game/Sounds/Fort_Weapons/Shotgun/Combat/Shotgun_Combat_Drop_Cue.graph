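# Sound cue that plays the drop sound for the combat shotgun weapon.
# A single wave player feeds the root directly with one drop sample, with pitch and volume lowered at the cue level.
# It uses the weapon drop attenuation, the loot collect sound class and a shared weapon-drop concurrency set, alongside a prevent-new override setting.
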SoundCue /Game/Sounds/Fort_Weapons/Shotgun/Combat/Shotgun_Combat_Drop_Cue
root=WavePlayer_1
nodes [{id=WavePlayer_1, properties {soundWaveAssetPtr=&/Shotgun_Combat_Drop01}}]
properties {
 attenuationSettings=&Fort_Weapon_Drop_Att
 maxDistance=3000
 pitchMultiplier=0.8
 soundClassObject=&Fort_Loot_Collect
 volumeMultiplier=0.45
 concurrencyOverrides {resolutionRule=PreventNew}
 concurrencySet [&Weapon_Drop_StopOldest]
}
