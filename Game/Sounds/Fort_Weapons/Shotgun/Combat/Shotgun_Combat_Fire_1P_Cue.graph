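# Sound cue for the combat shotgun's first-person fire sound.
# A quality-level root routes the first level to a random pick of three CloseNew waves at flat volume.
# The other three levels share a mixer, with only one input connected, feeding a volume-boosted random pick of two older-named Close waves.
# First-person weapon attenuation, fire sound class and gunfire concurrency settings govern playback.
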SoundCue /Game/Sounds/Fort_Weapons/Shotgun/Combat/Shotgun_Combat_Fire_1P_Cue
root=QualityLevel_0
nodes [
 {id=Mixer_1, children [Modulator_0,null,null], properties {inputVolume [1,0.8,0.8]}}
 {id=Modulator_0, children [Random_1], properties {volumeMax=1.3, volumeMin=1.25}}
 {id=Modulator_7, children [Random_0], properties {volumeMax=1, volumeMin=1}}
 {id=QualityLevel_0, children [Modulator_7,Mixer_1,Mixer_1,Mixer_1]}
 {id=Random_0, children [WavePlayer_0,WavePlayer_1,WavePlayer_2], properties {weights [1,1,1]}}
 {id=Random_1, children [WavePlayer_42, WavePlayer_43], properties {weights [1,1]}}
 {id=WavePlayer_0, properties {soundWaveAssetPtr=&/Shotgun_Combat_Fire_CloseNew_01}}
 {id=WavePlayer_1, properties {soundWaveAssetPtr=&/Shotgun_Combat_Fire_CloseNew_02}}
 {id=WavePlayer_2, properties {soundWaveAssetPtr=&/Shotgun_Combat_Fire_CloseNew_03}}
 {id=WavePlayer_42, properties {soundWaveAssetPtr=&/Shotgun_Combat_Fire_Close01}}
 {id=WavePlayer_43, properties {soundWaveAssetPtr=&/Shotgun_Combat_Fire_Close02}}
]
properties {
 attenuationSettings=&Fort_Weapon_Projectile_FirstPerson_Athena_Att
 maxDistance=2097152
 soundClassObject=&Fort_WeaponRanged_Fire_1P_PSM
 volumeMultiplier=1.1
 attenuationOverrides {falloffDistance=5000, attenuationShapeExtents {x=1000}}
 concurrencyOverrides {bLimitToOwner=true, maxCount=3}
 concurrencySet [&Weapon_Gunfire_FirstPerson_Conc]
}
