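# Sound cue for the shell-insert foley of the combat shotgun reload, using pump shotgun recordings.
# A Random node picks one of three equally weighted waves, then a Modulator holds volume steady while leaving default pitch variation, feeding a root Attenuation node.
# The cue uses the shared weapon reload attenuation and weapon foley sound class, with reduced volume and a concurrency override that prevents new sounds at the limit.
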SoundCue /Game/Sounds/Fort_Weapons/Shotgun/Combat/Shotgun_Combat_ReloadInsert_Cue
root=Attenuation_13
nodes [
 {id=Attenuation_13, children [Modulator_13], properties {attenuationOverrides {bSpatialize=false}}}
 {id=Modulator_13, children [Random_1], properties {volumeMax=1, volumeMin=1}}
 {id=Random_1, children [WavePlayer_3,WavePlayer_4,WavePlayer_5], properties {weights [1,1,1]}}
 {id=WavePlayer_3, properties {soundWaveAssetPtr=/Game/Sounds/Fort_Weapons/Shotgun/Reload_Elements/Pump_Shotgun_InsertShell_01}}
 {id=WavePlayer_4, properties {soundWaveAssetPtr=/Game/Sounds/Fort_Weapons/Shotgun/Reload_Elements/Pump_Shotgun_InsertShell_02}}
 {id=WavePlayer_5, properties {soundWaveAssetPtr=/Game/Sounds/Fort_Weapons/Shotgun/Reload_Elements/Pump_Shotgun_InsertShell_03}}
]
properties {
 attenuationSettings=&Fort_Weapon_Reload_Att
 bOverrideConcurrency=true
 maxDistance=2097152
 soundClassObject=&Fort_WeaponFoley
 volumeMultiplier=0.702
 attenuationOverrides {falloffDistance=1600}
 concurrencyOverrides {resolutionRule=PreventNew}
}
